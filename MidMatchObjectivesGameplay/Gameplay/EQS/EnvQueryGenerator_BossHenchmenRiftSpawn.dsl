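; EQS generator that supplies boss spawn patrol points from a henchmen rift spawner as query items.
; It takes the first context actor, casts it to B_MMObj_BossHenchmenRiftSpawner and reads its boss spawn patrol points.
; Each valid point is added as a generated actor.
; An empty context or failed cast only writes a shipping log message.
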(fn DoItemGenerationFromActors
  (params (ContextActors :type Array%20of%20Actor%20Object%20References))
  (bind _output (Utilities|Array|Get(acopy) ContextActors))
  ;[0.2708 0.07683 0.2357] Gather generator actor from the context actor.
  (Utilities|FlowControl|Sequence
    (:then_0
      (if (Utilities|Array|IsValidIndex ContextActors)
        (bind _asb_mmobj_boss_henchmen_rift_spawner (Utilities|Casting|CastToB_MMObj_BossHenchmenRiftSpawner _output)
          (:then
            (bind _patrolpoints (EQS|GetBossSpawnPatrolPoints _asb_mmobj_boss_henchmen_rift_spawner))
            (for _array_element _patrolpoints
              (Utilities|IsValid _array_element
                (:Is%20Valid
                  (EQS|AddGeneratedActor _array_element)))))
          (:CastFailed
            (Fortnite|FortShippingLog self (Utilities|String|Append "Context actor is not B_MMObj_BossHenchmenRiftSpawner class." (Utilities|GetDisplayName _output)) true)))
        (else
          (Fortnite|FortShippingLog self "ContextActors Is Empty." true))))
    (:then_1
      (return))))
